Material /Game/Effects/Niagara/MasterMaterials/Sprite/M_AshEmber_GPU_Particles
graph {
 properties {
  bEnableResponsiveAA=true
  bUsedWithNiagaraSprites=true
  bUsedWithParticleSprites=true
  blendMode=BLEND_Translucent
  pixelDepthOffsetMode=PDOM_Legacy
  shadingModel=MSM_Unlit
  usedShadingModels=Unlit
 }
 nodes [
  n1 = ParticleColor()
  n2 = TextureObject() {texture=/Game/Effects/Fort_Effects/Textures/Husk/T_Husk_Bits_SubUV}
  n3 = Add(A=ParticleRelativeTime(), B=ParticleRandom())
  n4 = ScalarParameter() {defaultValue=64, parameterName=Frame Speed}
  n5 = MaterialFunctionCall!(
   Texture=@n2
   UVs
   SubImages=Constant2Vector() {g=8, r=8}
   Frame=LinearInterpolate(
    A=Multiply(A=@n3, B=@n4)
    B=DynamicParameter() {paramNames [Frame,Param2,Param3,Param4]}.Frame
    Alpha=ScalarParameter() {parameterName=Drive UV via Dynamic Parameters}
   )
  ) {materialFunction=/Engine/Functions/Engine_MaterialFunctions03/Texturing/SubUV_Function}
  n6 = TextureSample(UVs=TextureCoordinate() {uTiling=0.125, vTiling=0.125}, Tex=@n2) {texture=&T_Husk_Bits_SubUV}
  n7 = ShadingPathSwitch(Default=Multiply(A=@n1.A, B=@n5.Alpha), Mobile=Multiply(A=@n1.A, B=@n6.A))
  n8 = ScalarParameter() {defaultValue=250, parameterName=Near Camera Distance}
  LinearInterpolate(
   A=@n7
   B=Multiply(A=@n7, B=MaterialFunctionCall(Distance=@n8) {materialFunction=&Near_Camera_Fade}.Result)
   Alpha=ScalarParameter() {parameterName=Use Near Camera Fade}
  )
  n9 = ComponentMask(None=@n5.RGB) {r=true}
  n10 = ComponentMask(None=@n5.RGB) {g=true}
  n11 = Constant() {r=0.1}
  n12 = Multiply(A=Add(A=ComponentMask(None=@n5.RGB) {b=true}, B=Multiply(A=@n10, B=@n11)), B=@n1.RGB)
  n13 = Add(A=@n6.G, B=Multiply(A=@n6.B) {constB=0.2})
  MaterialFunctionCall(Emissive=ShadingPathSwitch(Default=Add(A=@n9, B=@n12), Mobile=Multiply(A=@n1.RGB, B=@n13))) {materialFunction=&MF_FX_EmissiveAdjustments}
 ]
}
