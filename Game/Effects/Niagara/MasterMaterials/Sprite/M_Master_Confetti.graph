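# Master material for confetti particles on Niagara sprites and meshes, two-sided lit with metallic, low-roughness defaults.
# Rotates vertices around the particle position by an angle scaled along the V coordinate, twisting each piece.
# Dynamic parameters supply angle and axis, and FixRotateAboutAxisNormals receives the same rotation to correct normals.
# Particle color multiplied by time-of-day tints and a UseEmissive? scalar feeds the shared emissive adjustment function.
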Material /Game/Effects/Niagara/MasterMaterials/Sprite/M_Master_Confetti
graph {
 properties {
  bUsedWithNiagaraMeshParticles=true
  bUsedWithNiagaraSprites=true
  bUsedWithParticleSprites=true
  pixelDepthOffsetMode=PDOM_Legacy
  translucencyLightingMode=TLM_VolumetricPerVertexDirectional
  twoSided=true
  usedShadingModels=Default Lit
 }
 nodes [
  n1 = DynamicParameter() {defaultValue=<rgba 0.1 1 1 1>, paramNames ["Angle (0-1)", X, Y, Z]}
  n2 = AppendVector(A=@n1.X, B=@n1.X)
  n3 = Normalize(VectorInput=Transform(None=AppendVector(A=@n1.Z, B=@n2)) {transformSourceType=TRANSFORMSOURCE_Local})
  n4 = ComponentMask(None=TextureCoordinate()) {g=true}
  n5 = Multiply(
   A=Subtract(A=Multiply(A=@n4) {constB=2})
   B=DynamicParameter() {defaultValue=<rgba 0.1 1 1 1>, paramNames ["Angle (0-1)", X, Y, Z]}."Angle (0-1)"
  )
  MaterialFunctionCall!(InVector=RotateAboutAxis(
    NormalizedRotationAxis=NamedRerouteDeclaration(None=@n3) {name=Axis, nodeColor=<rgba 1 0 0.1641 1>}
    RotationAngle=NamedRerouteDeclaration(None=@n5) {name=ANgle, nodeColor=<rgba 0.125 1 0 1>}
    PivotPoint=ParticlePositionWS()
    Position=WorldPosition() {worldPositionShaderOffset=WPT_ExcludeAllShaderOffsets}.XYZ
   )) {materialFunction=/Game/Athena/Environments/Landscape/Foliage/MF_MakePreciseVec3}
  n6 = MaterialFunctionCall(
   Color_Morning=VectorParameter() {parameterName=ColorMorning, defaultValue=<rgba 0.8 0.8 0.8 0>}.RGB
   Color_Day=VectorParameter() {parameterName=ColorDay, defaultValue=<rgba 1 1 1 0>}.RGB
   Color_Evening=VectorParameter() {parameterName=ColorEvening, defaultValue=<rgba 0.8 0.8 0.8 0>}.RGB
   Color_Night=VectorParameter() {parameterName=ColorNight, defaultValue=<rgba 0.5 0.5 0.5 0>}.RGB
  ) {materialFunction=&Color_Time_of_Day}
  n7 = ParticleColor()
  MaterialFunctionCall(Emissive=Multiply(A=ScalarParameter() {parameterName=UseEmissive?}, B=Multiply(A=@n6.Result, B=@n7.RGB))) {materialFunction=&MF_FX_EmissiveAdjustments}
  MaterialFunctionCall(Rotation Axis=NamedRerouteUsage(), Rotation Angle=NamedRerouteUsage(), Normals Pivot=ParticlePositionWS()) {materialFunction=&FixRotateAboutAxisNormals}
  ScalarParameter() {defaultValue=1, parameterName=Metallic}
  ScalarParameter() {defaultValue=1, parameterName=Specular}
  ScalarParameter() {defaultValue=0.02, parameterName=Roughness}
 ]
}
