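# Translucent, two-sided master material for Niagara confetti sprites, lit with per-vertex directional volumetric lighting.
# Particle color is tinted by morning, day, evening and night parameters through a time-of-day function, and separately scaled by UseEmissive? into an emissive adjustment function.
# A selectable channel of a 4x4 SubUV confetti texture, particle alpha and a near-camera fade multiply into a mask.
# A second confetti texture is sampled as a normal map.
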Material /Game/Effects/Niagara/MasterMaterials/Sprite/M_Master_Confetti_TwoSided_Simple
graph {
 properties {
  bUsedWithNiagaraSprites=true
  blendMode=BLEND_Translucent
  pixelDepthOffsetMode=PDOM_Legacy
  translucencyLightingMode=TLM_VolumetricPerVertexDirectional
  twoSided=true
  usedShadingModels=Default Lit
 }
 nodes [
  n1 = ParticleColor()
  MaterialFunctionCall(Emissive=Multiply(A=ScalarParameter() {parameterName=UseEmissive?}, B=@n1.RGB)) {materialFunction=&MF_FX_EmissiveAdjustments}
  Multiply(
   A=MaterialFunctionCall(
    Color_Morning=VectorParameter() {parameterName=ColorMorning, defaultValue=<rgba 0.8 0.8 0.8 0>}.RGB
    Color_Day=VectorParameter() {parameterName=ColorDay, defaultValue=<rgba 1 1 1 0>}.RGB
    Color_Evening=VectorParameter() {parameterName=ColorEvening, defaultValue=<rgba 0.8 0.8 0.8 0>}.RGB
    Color_Night=VectorParameter() {parameterName=ColorNight, defaultValue=<rgba 0.5 0.5 0.5 0>}.RGB
   ) {materialFunction=&Color_Time_of_Day}.Result
   B=@n1.RGB
  )
  n2 = TextureSample() {texture=/Game/Effects/Fort_Effects/Textures/Confetti/T_Confetti_SubUV_4X4_A}
  Multiply(
   A=Multiply(A=@n1.A, B=StaticComponentMaskParameter(None=@n2.B) {defaultB=true, parameterName=PickChannel})
   B=MaterialFunctionCall(Distance=ScalarParameter() {defaultValue=15, parameterName=Nearfade_Distance}) {materialFunction=&Near_Camera_Fade}.Result
  )
  TextureSample() {samplerType=SAMPLERTYPE_Normal, texture=/Game/Effects/Fort_Effects/Textures/Confetti/T_Confetti_SubUV_4X4_03}
 ]
}
