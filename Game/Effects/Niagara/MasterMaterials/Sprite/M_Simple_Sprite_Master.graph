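# Translucent unlit master material for Niagara sprites, ribbons and mesh particles, covering simple sprites, flares and panning effects.
# Scalar parameters blend between modes via lerps rather than static switches: textured or gradient sprite, procedural flare, and two-layer panner.
# Opacity layers particle alpha with depth, view-angle and V-axis fades, plus sine-based flicker in flare mode.
# Emissive applies particle colour, optional time-of-day tint and HDR-from-LDR remapping, then shared FX adjustments.
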Material /Game/Effects/Niagara/MasterMaterials/Sprite/M_Simple_Sprite_Master
graph {
 properties {
  bUsedWithNiagaraMeshParticles=true
  bUsedWithNiagaraRibbons=true
  bUsedWithNiagaraSprites=true
  blendMode=BLEND_Translucent
  pixelDepthOffsetMode=PDOM_Legacy
  shadingModel=MSM_Unlit
  usedShadingModels=Unlit
 }
 nodes [
  n1 = ComponentMask(None=TextureCoordinate()) {g=true}
  n2 = LinearInterpolate(A=@n1, B=OneMinus(None=@n1), Alpha=ScalarParameter() {group=FADE PARAMS, parameterName=FLIP DIRECTION})
  n3 = ScalarParameter() {defaultValue=2, group=FADE PARAMS, parameterName=FADE POWER}
  n4 = MaterialFunctionCall("FadeSensitivity (optional)"=ScalarParameter() {parameterName=ViewAngleFade}) {materialFunction=&FadeBasedOnViewAngle}
  n5 = ScalarParameter() {parameterName=UseViewAngleFade}
  n6 = ParticleColor()
  n7 = Sine(None=ComponentMask(None=TextureCoordinate()) {g=true}) {period=2}
  n8 = ScalarParameter() {defaultValue=2, parameterName=Panner Gradient Exponent}
  n9 = Multiply(A=TextureCoordinate(), B=ScalarParameter() {defaultValue=1, parameterName=Panner Tile A})
  n10 = AppendVector(A=ScalarParameter() {defaultValue=10, parameterName=Panner Speed A}, B=Constant())
  n11 = Multiply(
   A=Saturate(None=Power(Base=@n7, Exp=@n8))
   B=TextureSampleParameter2D(UVs=Panner(Coordinate=@n9, Speed=@n10) {speedX=10}) {parameterName=Panner Texture A, texture=/Game/Effects/Fort_Effects/Textures/Smoke/T_Smoke_Trail_Soft_2}.R
  )
  n12 = TextureCoordinate() {uTiling=4}
  n13 = ScalarParameter() {defaultValue=1, parameterName=Panner Tile B}
  n14 = ScalarParameter() {defaultValue=10, parameterName=Panner Speed B}
  n15 = Constant()
  n16 = TextureSampleParameter2D(UVs=Panner(Coordinate=Multiply(A=@n12, B=@n13), Speed=AppendVector(A=@n14, B=@n15)) {speedX=10}) {
   parameterName=Panner Texture B
   samplerType=SAMPLERTYPE_Grayscale
   texture=/Game/Effects/Fort_Effects/Textures/Smoke/T_Smoke_Trail_Soft_NoClamp
  }
  n17 = Add(A=@n11, B=Power(Base=@n16.R, Exp=ScalarParameter() {defaultValue=1.3, parameterName=Power}) {constExponent=1.3})
  n18 = Multiply(A=@n6.A, B=@n17)
  n19 = ScalarParameter() {parameterName=Use_custom_fade_flare}
  n20 = ParticleColor()
  n21 = Multiply(
   A=Power(
    Base=TextureSampleParameter2D() {parameterName=Sprite Texture, texture=/Game/Effects/Fort_Effects/Textures/Sparks/T_Sparkles_4Star_01_D}.R
    Exp=ScalarParameter() {defaultValue=1, parameterName=Sprite_Texture_Power}
   )
   B=ScalarParameter() {defaultValue=1, parameterName=Sprite_Texture_intensity}
  )
  n22 = ComponentMask(None=TextureCoordinate()) {g=true}
  n23 = Saturate(None=Power(Base=OneMinus(None=@n22)) {constExponent=1.6})
  n24 = ScalarParameter() {parameterName=Use_sample_Gradient_mask}
  n25 = ParticleColor()
  n26 = DynamicParameter() {desc=Index 1 Params for Niagara, parameterIndex=1, defaultValue=<rgba 30 1 1 1>, paramNames [DepthFadeDist, Param2, Param3, Param4]}
  n27 = ScalarParameter() {defaultValue=30, parameterName=DepthFade}
  n28 = DepthFade(Opacity=@n25.A, FadeDistance=If(A=@n26.DepthFadeDist, B=Constant() {r=1}, "A > B"=@n26.DepthFadeDist, "A == B"=@n27, "A < B"=@n27))
  n29 = ShadingPathSwitch(Default=@n28, Deferred=@n28, Mobile=@n25.A) {desc=Migrated from FeatureLevelSwitch}
  n30 = DynamicParameter() {defaultValue=<rgba 1 1 1 0>, paramNames [FlickerFade, FlickerSpeed, ColorJitter, FadeBasedOnViewAngle]}
  n31 = Multiply(A=Time(), B=@n30.FlickerSpeed)
  n32 = Sine(None=@n31) {period=0.075}
  n33 = Multiply(A=Multiply(A=Sine(None=@n31) {period=0.155}, B=Sine(None=@n31) {period=0.2}), B=Sine(None=@n31) {period=0.43})
  n34 = LinearInterpolate(A=Clamp(None=Multiply(A=@n32, B=@n33)) {maxDefault=3}, Alpha=Saturate(None=@n30.FlickerFade)) {constA=1}
  n35 = DynamicParameter() {defaultValue=<rgba 1 1 1 0>, paramNames [FlickerFade, FlickerSpeed, ColorJitter, FadeBasedOnViewAngle]}
  n36 = LinearInterpolate(
   A=Multiply(A=@n20.A, B=LinearInterpolate(A=@n21, B=@n23, Alpha=@n24))
   B=Multiply(
    A=Multiply(A=@n34, B=@n29)
    B=MaterialFunctionCall("FadeSensitivity (optional)"=@n35.FadeBasedOnViewAngle) {materialFunction=&FadeBasedOnViewAngle}.Result
   )
   Alpha=@n19
  )
  n37 = ScalarParameter() {parameterName=Use Multiply Simple Panner With Texture}
  n38 = LinearInterpolate(A=@n18, B=Multiply(A=@n36, B=@n18), Alpha=@n37)
  n39 = ScalarParameter() {parameterName=Use_simple_panner}
  n40 = DepthFade(
   Opacity=Multiply(A=LinearInterpolate(B=@n4.Result, Alpha=@n5) {constA=1}, B=LinearInterpolate(A=@n36, B=@n38, Alpha=@n39))
   FadeDistance=ScalarParameter() {defaultValue=10, group="1) Depth Fade", parameterName=Overall Depth Fade}
  )
  LinearInterpolate(A=@n40, B=Multiply(A=@n40, B=Power(Base=@n2, Exp=@n3)), Alpha=ScalarParameter() {group=FADE PARAMS, parameterName=FRONT-END FADE})
  LinearInterpolate(A=@n11, B=@n16.R) {constAlpha=0.1}
  n41 = VectorParameter() {parameterName=Color_morning, defaultValue=<rgba 1 1 1 1>}
  n42 = VectorParameter() {parameterName=Color_Day, defaultValue=<rgba 1 1 1 1>}
  n43 = VectorParameter() {parameterName=Color_Evening, defaultValue=<rgba 1 1 1 1>}
  n44 = VectorParameter() {parameterName=Color_Night, defaultValue=<rgba 1 1 1 1>}
  n45 = Multiply(A=@n6.RGB, B=@n17)
  n46 = Constant() {r=0.5}
  n47 = SphereMask(A=TextureCoordinate(), B=@n46, Radius=@n46) {hardnessPercent=0}
  n48 = Multiply(
   A=ScalarParameter() {defaultValue=2, parameterName=CenterBrightness}
   B=Power(Base=@n47, Exp=ScalarParameter() {defaultValue=25, parameterName=CentertPower}) {constExponent=25}
  ) {constB=2}
  n49 = Desaturation(None=@n25.RGB, Fraction=ScalarParameter() {defaultValue=0.3, parameterName=CenterDesaturation})
  n50 = ScalarParameter() {defaultValue=4, parameterName=EdgePower}
  n51 = Add(
   A=Multiply(A=Power(Base=@n47, Exp=@n50) {constExponent=4}, B=ScalarParameter() {defaultValue=0.15, parameterName=EdgeBrightness}) {constB=0.15}
   B=TextureSampleParameter2D() {parameterName=FlareMask, texture=/Game/Effects/Fort_Effects/Textures/Flares/T_Flare_Spikeball}.RGB
  )
  n52 = LinearInterpolate(A=@n20.RGB, B=Add(A=Multiply(A=@n48, B=@n49), B=Multiply(A=@n51, B=@n25.RGB)), Alpha=@n19)
  n53 = Multiply(A=@n52, B=@n45)
  n54 = Multiply(
   A=LinearInterpolate(
    A=Constant() {r=1}
    B=MaterialFunctionCall(Color_Morning=@n41.RGB, Color_Day=@n42.RGB, Color_Evening=@n43.RGB, Color_Night=@n44.RGB) {materialFunction=&Color_Time_of_Day}.Result
    Alpha=ScalarParameter() {parameterName=Use_color_time_of_day}
   )
   B=LinearInterpolate(A=@n52, B=LinearInterpolate(A=@n45, B=@n53, Alpha=@n37), Alpha=@n39)
  )
  n55 = Saturate(None=@n54)
  n56 = Multiply(A=ParticleColor().RGB, B=@n55)
  n57 = ScalarParameter() {defaultValue=1, parameterName=HDR Multiplier}
  MaterialFunctionCall(Emissive=LinearInterpolate(
    A=@n54
    B=MaterialFunctionCall(LDR Input=@n55, HDR Tint=@n56, HDR Intensity Multiplier=@n57) {materialFunction=&DeriveHDRfromLDR}.RemappedRange
    Alpha=ScalarParameter() {parameterName=Derive HDR from LDR}
   )) {materialFunction=&MF_FX_EmissiveAdjustments}
 ]
}
