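# MF_Erosion is a material function that turns an input value into a hard-edged erosion mask for Niagara effect materials.
# It subtracts the Erosion output of MF_Niagara_Erosion from the Erosion Value input and saturates the result.
# A large constant multiply and second saturate sharpen that into a near-binary cutoff.
# The input is typed as Material Attributes, and a GetMaterialAttributes node reading it does not feed the output.
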MaterialFunction /Game/Effects/Niagara/MasterMaterials/Functions/MF_Erosion
graph {nodes [
  n1 = FunctionInput() {description=Input value to clamp, inputName=Erosion Value, inputType=FunctionInput_MaterialAttributes}
  n2 = Saturate(None=Subtract(A=@n1, B=MaterialFunctionCall() {materialFunction=&/MF_Niagara_Erosion}.Erosion))
  FunctionOutput(None=Saturate(None=Multiply(A=@n2) {constB=100}))
  GetMaterialAttributes(None=@n1) {attributeGetTypes [#1, #2]}
 ]}
