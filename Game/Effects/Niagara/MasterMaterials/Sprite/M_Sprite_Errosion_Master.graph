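# Master unlit translucent material for Niagara sprites with switchable erosion, distortion and color remapping.
# A panned normal or grayscale map combines with inverted masked ColorMap strength, driving a two-color remap and Niagara-controlled erosion opacity.
# A panned noise texture, optionally in polar coordinates, distorts the ColorMap mask UVs, bypassed on mobile feature levels.
# Particle color, hue shift and emissive adjustments finish the color; particle alpha scales opacity.
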Material /Game/Effects/Niagara/MasterMaterials/Sprite/M_Sprite_Errosion_Master
graph {
 properties {bUsedWithNiagaraSprites=true, blendMode=BLEND_Translucent, pixelDepthOffsetMode=PDOM_Legacy, shadingModel=MSM_Unlit, usedShadingModels=Unlit}
 nodes [
  n1 = ParticleColor()
  n2 = TextureSampleParameter2D(UVs=MaterialFunctionCall!(Texture Coordinates=TextureCoordinate(), U Frequency=Constant() {r=1}, V Frequency, Offset=Constant() {r=1}, OffsetMultiplier) {materialFunction=/Game/Effects/Niagara/MasterMaterials/Functions/MF_Niagara_Panner}.Result) {
   desc=This should be a normalmap
   group=Erosion
   parameterName=ErosionMap
   samplerType=SAMPLERTYPE_Normal
   texture=/Game/Effects/Fort_Effects/Textures/Water/T_water_splash_04_n
  }
  n3 = OneMinus(None=DotProduct(A=@n2.RGB, B=Constant3Vector() {constant=<rgba 0 0 1 0>}))
  n4 = ScalarParameter() {defaultValue=1, group=Erosion, parameterName=Exponent, sortPriority=5}
  n5 = TextureCoordinate()
  n6 = Subtract(A=@n5) {constB=0.5}
  n7 = Arctangent2Fast(Y=ComponentMask(None=@n6) {r=true}, X=ComponentMask(None=@n6) {g=true})
  n8 = AppendVector(
   A=Add(A=Divide(A=@n7) {constB=6.28})
   B=SquareRoot(None=MaterialFunctionCall(Vector 2=@n6) {materialFunction=/Engine/Functions/Engine_MaterialFunctions02/Utility/VectorLength}.V2 Length)
  )
  n9 = TextureSampleParameter2D(UVs=MaterialFunctionCall(Texture Coordinates=StaticSwitchParameter(True=@n8, False=@n5) {defaultValue=true, group=UVs, parameterName=Enable Polar Coordinates}) {materialFunction=&MF_Niagara_Panner}.Result) {
   desc=The strenght of the distortion is controled through Niagara Moduels
   group=Distortion
   parameterName=Distortion Texture
   sortPriority=1
   texture=/Game/Global/Textures/Default/Tileable_Patterns/T_Tilling_Distortion_Map_01
  }
  n10 = TextureCoordinate()
  n11 = Add(
   A=Multiply(
    A=ComponentMask(None=@n9.RGB) {g=true, r=true}
    B=MaterialFunctionCall() {materialFunction=/Game/Effects/Niagara/MasterMaterials/Functions/MF_Niagara_MaterialPannerParameters}.Amplitude
   )
   B=@n10
  )
  n12 = TextureSampleParameter2D!(
   UVs=StaticSwitchParameter(True=FeatureLevelSwitch(Default=@n11, ES2=@n10, ES3_1=@n10, SM4_REMOVED=@n11), False=@n10) {desc=This is 40 instructions with radial uvs., group=Distortion, parameterName=Enable Distortion}
   Level
   Apply View MipBias
  ) {desc=o, group=Color, mipValueMode=TMVM_MipLevel, parameterName=ColorMap, texture=/Game/Effects/Fort_Effects/Textures/Water/T_water_splash_04}
  n13 = MaterialFunctionCall(Vector 3=Multiply(A=VectorParameter() {group=Color, parameterName=Color Mask, sortPriority=1, defaultValue=<rgba 1 1 1 1>}.RGB, B=@n12.RGB)) {materialFunction=&VectorLength}
  n14 = Add(
   A=StaticSwitchParameter(True=Power(Base=@n3, Exp=@n4), False=@n2.RGB) {defaultValue=true, group=Erosion, parameterName=isNormalMap}
   B=OneMinus(None=Saturate(None=@n13.V3 Length))
  )
  n15 = TextureSampleParameter2D() {desc=o, group=Color, mipValueMode=TMVM_MipLevel, parameterName=ColorMap, texture=&T_water_splash_04}
  n16 = StaticSwitchParameter(
   True=LinearInterpolate(
    A=VectorParameter() {group=Color, parameterName=A, sortPriority=2, defaultValue=<rgba 0.6875 0.0342 0 1>}.RGB
    B=VectorParameter() {group=Color, parameterName=B, sortPriority=2, defaultValue=<rgba 1.1 0.4029 0.0338 1>}.RGB
    Alpha=@n14
   )
   False=@n15.RGB
  ) {
   desc=Disabling this just does a color map. Enabling this colors based on the normal map. This is an additional 19 instructions.
   group=Color
   parameterName=RemapColor
   sortPriority=1
  }
  MaterialFunctionCall(Emissive=MaterialFunctionCall(Hue Shift Percentage=ScalarParameter() {group=Color, parameterName=HueShift, sortPriority=5}, Texture=Multiply(A=@n1.RGB, B=@n16)) {materialFunction=&HueShift}.Result) {materialFunction=&MF_FX_EmissiveAdjustments}
  n17 = MaterialFunctionCall!(Erosion Value=OneMinus(None=@n14)) {
   desc="The erosion is controled from  the Niagara Module Material Erosion."
   materialFunction=/Game/Effects/Niagara/MasterMaterials/Functions/MF_Erosion
  }
  n18 = ParticleColor()
  StaticSwitchParameter(True=Multiply(A=@n17.Result, B=@n18.A), False=Multiply(A=@n18.A, B=@n13.V3 Length)) {desc=Enable Erosion. This is 39 Instructions., group=Erosion, parameterName=Enable}
 ]
}
