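# Material function exposed under FX Uvs that selects, scales and pans texture UVs for Niagara effect materials.
# An input blends UV channel 0 to 1, then U and V frequency multipliers scale the result.
# The "Scale UVs From Center" parameter blends toward center-based scaling, which uses only the U multiplier.
# A Panner offsets the result with U and V speed inputs, timed by UV Time from the MF_NFX_Bridge Niagara bridge function.
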MaterialFunction /Game/Effects/Niagara/MasterMaterials/Functions/UVs/MF_NFX_TextureUVPannerv2
graph {
 properties {bExposeToLibrary=true, description=UV Channels with Niagara Bridge, libraryCategoriesText [FX Uvs]}
 nodes [
  n1 = FunctionInput() {
   bUsePreviewValueAsDefault=true
   description="This allows you to choose either UV channel  0 or 1"
   inputName=UV Channel
   inputType=FunctionInput_Scalar
   sortPriority=1
  }
  n2 = ComponentMask(None=LinearInterpolate(A=TextureCoordinate(), B=TextureCoordinate() {coordinateIndex=1}, Alpha=Clamp(None=@n1))) {g=true, r=true}
  n3 = FunctionInput() {
   bUsePreviewValueAsDefault=true
   description=This is an independent multiplier on the incoming frequency
   inputName=U Freq Multiplier
   inputType=FunctionInput_Scalar
   sortPriority=2
   previewValue=<1 0 0 1>
  }
  n4 = FunctionInput() {
   bUsePreviewValueAsDefault=true
   description=This is an independent multiplier on the frequency
   inputName=V Freq Multiplier
   inputType=FunctionInput_Scalar
   sortPriority=3
   previewValue=<1 0 0 1>
  }
  n5 = Multiply(A=@n2, B=AppendVector(A=@n3, B=@n4)) {desc=This might need a frac option for getting textures to tile. This scales the texture.}
  n6 = MaterialFunctionCall(UVs=@n2, Texture Scale=@n3) {materialFunction=&ScaleUVsByCenter}
  n7 = ScalarParameter() {group="9)UV Options", parameterName=Scale UVs From Center}
  n8 = FunctionInput() {
   bUsePreviewValueAsDefault=true
   description=A multiplier on the speed of the pan.
   inputName=Pan Speed U
   inputType=FunctionInput_Scalar
   sortPriority=4
  }
  n9 = FunctionInput() {bUsePreviewValueAsDefault=true, description=A multiplier on the speed of the pan., inputName=Pan Speed V, sortPriority=5}
  FunctionOutput(None=Panner(
    Coordinate=LinearInterpolate(A=@n5, B=@n6.UVs, Alpha=@n7)
    Time=MaterialFunctionCall() {desc=This is to communicate with Niagara., materialFunction=/Game/Effects/Niagara/MasterMaterials/Functions/MF_NFX_Bridge}.UV Time
    Speed=AppendVector(A=@n8, B=@n9)
   ) {desc=This offsets the texture.})
 ]
}
