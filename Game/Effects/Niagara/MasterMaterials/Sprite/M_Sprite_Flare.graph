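# M_Sprite_Flare is an unlit translucent master material for Niagara sprite flare effects, built on material attributes and driven through MF_NFX_Bridge.
# A panning distortion map, carried in BentNormal, offsets the base texture, which collapses to its brightest channel.
# Opacity comes from that value or an eroded separate texture, shaped by particle alpha, ribbon rounding and a V gradient.
# Depth fade, colorize and a V-based hue shift finish it, with optional eye-adaptation compensation on emissive.
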Material /Game/Effects/Niagara/MasterMaterials/Sprite/M_Sprite_Flare
graph {
 properties {
  bUseMaterialAttributes=true
  bUsedWithNiagaraSprites=true
  blendMode=BLEND_Translucent
  pixelDepthOffsetMode=PDOM_Legacy
  shadingModel=MSM_Unlit
  usedShadingModels=Unlit
 }
 nodes [
  n1 = ScalarParameter() {
   defaultValue=1
   desc=How much the distortion should add to Channel UV0.
   group="1.2) Distortion UV Influence"
   parameterName=UV0 Distortion Influence
   sliderMax=1
  }
  n2 = LinearInterpolate(
   A=TextureCoordinate()
   B=TextureCoordinate() {coordinateIndex=1}
   Alpha=ScalarParameter() {group="1.1) Distortion UVs", parameterName="UV Channel (Distortion)", sliderMax=1}
  )
  n3 = AppendVector(
   A=ScalarParameter() {defaultValue=1, group="1.1) Distortion UVs", parameterName="U Freq Mult (Distort)", sortPriority=1}
   B=ScalarParameter() {defaultValue=1, group="1.1) Distortion UVs", parameterName="V Freq Mult (Distort)", sortPriority=2}
  )
  n4 = ScalarParameter() {defaultValue=1, group="1.1) Distortion UVs", parameterName="Pan U Speed (Distort)", sortPriority=3}
  n5 = ScalarParameter() {group="1.1) Distortion UVs", parameterName="Pan V Speed (Distort)", sortPriority=4}
  n6 = TextureSampleParameter2D(UVs=Panner(
    Coordinate=Multiply(A=@n2, B=@n3) {desc=This might need a frac option for getting textures to tile. This scales the texture.}
    Time=MaterialFunctionCall() {desc=This is to communicate with Niagara., materialFunction=/Game/Effects/Niagara/MasterMaterials/Functions/MF_NFX_Bridge}.Distortion UV Time
    Speed=AppendVector(A=@n4, B=@n5)
   ) {desc=This offsets the texture.}) {
   group="1) Distortion"
   parameterName=DistortionMap
   samplerType=SAMPLERTYPE_Normal
   sortPriority=1
   texture=&T_Weap_Distort_N_02
  }
  n7 = GetMaterialAttributes(None=SetMaterialAttributes!(MaterialAttributes, BentNormal=Multiply(A=@n1, B=@n6.RGB)) {attributeSetTypes [#1]}) {attributeGetTypes [#2, #1]}
  n8 = ScalarParameter() {group="2) BaseColor", parameterName="UV Channel (BaseColor)", sliderMax=1, sortPriority=1}
  n9 = ScalarParameter() {defaultValue=1, group="2) BaseColor", parameterName="U Freq (BaseColor)", sortPriority=2}
  n10 = ScalarParameter() {defaultValue=1, group="2) BaseColor", parameterName="V Freq (BaseColor)", sortPriority=3}
  n11 = ScalarParameter() {group="2) BaseColor", parameterName="Pan U Speed (BaseColor)", sortPriority=4}
  n12 = ScalarParameter() {group="2) BaseColor", parameterName="Pan V Speed (BaseColor)", sortPriority=5}
  n13 = TextureSampleParameter2D(UVs=Add(
    A=ComponentMask(None=@n7.BentNormal) {g=true, r=true}
    B=MaterialFunctionCall!(UV Channel=@n8, U Freq Multiplier=@n9, V Freq Multiplier=@n10, Pan Speed U=@n11, Pan Speed V=@n12) {materialFunction=/Game/Effects/Niagara/MasterMaterials/Functions/UVs/MF_NFX_TextureUVPannerv2}.Result
   ) {desc=Distortion}) {
   desc=This texture provides the foundation for the other channels like opacity and emissive color.
   group="2) BaseColor"
   parameterName=Texture
   samplerType=SAMPLERTYPE_Normal
   texture=/Game/Effects/Fort_Effects/Textures/Tile/MuzzlePattern_Lightning_TileX
  }
  n14 = Add(A=@n7.Base Color, B=@n13.RGB)
  n15 = SetMaterialAttributes!(MaterialAttributes=@n7.MaterialAttributes, Base Color=@n14, Emissive Color) {attributeSetTypes [#2, #3]}
  n16 = GetMaterialAttributes(None=@n15) {attributeGetTypes [#2]}
  n17 = MaterialFunctionCall() {materialFunction=&MF_NFX_Bridge}
  n18 = MaterialFunctionCall(Color=Multiply(
    A=@n16.Base Color
    B=MaterialFunctionCall(X=@n17.R Component Mask, Y=@n17.R Component Mask, Z=@n17.R Component Mask) {materialFunction=&MakeFloat3}.Result
   )) {materialFunction=&SplitComponents}
  n19 = Max(A=@n18.R, B=@n18.R)
  n20 = GetMaterialAttributes(None=SetMaterialAttributes!(MaterialAttributes=@n16.MaterialAttributes, Base Color=Max(A=@n19, B=@n18.B)) {attributeSetTypes [#2]}) {attributeGetTypes [
    #4
    #5
    #2
    #1
   ]}
  ComponentMask(None=@n20.BentNormal) {g=true, r=true}
  ComponentMask(None=@n6.RGB) {g=true, r=true}
  MaterialFunctionCall!(In=@n15) {materialFunction=/Game/Effects/Niagara/MasterMaterials/Functions/MF_NFX_ComponentMask}
  MaterialFunctionCall() {materialFunction=&HueShift}
  n21 = ScalarParameter() {group="2.1) Opacity Uvs", parameterName="UV Channel (Opacity)", sliderMax=1, sortPriority=1}
  n22 = ScalarParameter() {defaultValue=1, group="2.1) Opacity Uvs", parameterName="U Freq (Opacity)", sortPriority=2}
  n23 = ScalarParameter() {defaultValue=1, group="2.1) Opacity Uvs", parameterName="V Freq (Opacity)", sortPriority=3}
  n24 = ScalarParameter() {defaultValue=1, group="2.1) Opacity Uvs", parameterName="Pan U Speed (Opacity)", sortPriority=4}
  n25 = ScalarParameter() {group="2.1) Opacity Uvs", parameterName="Pan V Speed (Opacity)", sortPriority=5}
  n26 = TextureSampleParameter2D(UVs=Add(B=MaterialFunctionCall(UV Channel=@n21, U Freq Multiplier=@n22, V Freq Multiplier=@n23, Pan Speed U=@n24, Pan Speed V=@n25) {materialFunction=&MF_NFX_TextureUVPannerv2}.Result) {desc=Distortion}) {
   group="2) Opacity"
   parameterName=Opacity Texture
   sortPriority=1
   texture=/Game/Effects/Fort_Effects/Textures/Tile/T_Tile_Texture_Cartoon_Flames_Layer
  }
  n27 = ChannelMaskParameter(None=@n26.RGB) {group="2) Opacity", parameterName=Opacity Texture Mask, sortPriority=2}
  n28 = ComponentMask(None=TextureCoordinate()) {g=true}
  n29 = MaterialFunctionCall() {materialFunction=&MF_NFX_Bridge}
  n30 = Subtract(A=@n27.R, B=Multiply(A=OneMinus(None=@n28), B=@n29.Add))
  n31 = Saturate(None=@n30)
  n32 = StaticSwitchParameter(
   True=@n20.Base Color
   False=LinearInterpolate(
    A=@n31
    B=Multiply(A=@n20.Base Color, B=@n31)
    Alpha=ScalarParameter() {group="2) Opacity", parameterName="Multipy Opacity Texture Against Base Color)", sliderMax=1}
   )
  ) {
   defaultValue=true
   desc="If Activated, you can use a seperate texture read for your opacity. Otherwise, it inherits from the basecolor material attribute."
   group="2) Opacity"
   parameterName=Use Base Color for Opacity
  }
  n33 = Multiply(A=Saturate(None=@n32), B=ParticleColor().A)
  n34 = GetMaterialAttributes(None=SetMaterialAttributes!(MaterialAttributes=@n20.MaterialAttributes, Opacity=@n33, Opacity Mask=@n33) {attributeSetTypes [#4, #5]}) {attributeGetTypes [#4, #6, #7]}
  n35 = TextureCoordinate()
  n36 = Multiply(A=Subtract(A=ComponentMask(None=@n35) {r=true}, B=Constant() {r=0.5}), B=Constant() {r=2})
  n37 = OneMinus(None=Saturate(None=Multiply(A=@n36, B=@n36)))
  n38 = OneMinus(None=ComponentMask(None=@n35) {desc=UV0 G seems broke., g=true})
  n39 = ScalarParameter() {
   defaultValue=0.5
   desc=Amount of rounding on the base of the ribbon
   group="2) Opacity"
   parameterName=Ribbon Rounding
   sliderMax=1
   sortPriority=1
  }
  n40 = OneMinus(None=Saturate(None=Divide(A=@n38, B=@n39)))
  n41 = Multiply(A=@n40, B=@n40)
  n42 = ScalarParameter() {defaultValue=0.2, desc=Softness, group="2) Opacity", parameterName=Ribbon Rounding Softness, sliderMax=1, sortPriority=2}
  n43 = OneMinus(None=Abs(None=@n36))
  n44 = Multiply(A=@n43, B=@n43)
  n45 = Saturate(None=Divide(A=Subtract(A=@n37, B=@n41), B=LinearInterpolate(A=@n42, Alpha=@n44)))
  n46 = GetMaterialAttributes(None=SetMaterialAttributes!(MaterialAttributes=@n34.MaterialAttributes, Opacity=Multiply(A=@n34.Opacity, B=@n45)) {attributeSetTypes [#4]}) {attributeGetTypes [#4]}
  n47 = MaterialFunctionCall() {materialFunction=&MF_NFX_Bridge}
  n48 = Subtract(A=ComponentMask(None=TextureCoordinate()) {g=true}, B=@n47.Add)
  n49 = GetMaterialAttributes(None=SetMaterialAttributes!(MaterialAttributes=@n46.MaterialAttributes, Base Color, Opacity=Multiply(A=@n46.Opacity, B=@n48), Emissive Color) {attributeSetTypes [#2, #4, #3]}) {attributeGetTypes [#2]}
  n50 = Add(A=@n49.Base Color, B=@n47.Add)
  Max(A=Power(Base=Multiply(A=@n50, B=@n47.PostMult), Exp=@n47.Pow) {constExponent=0.5}) {constB=0}
  n51 = ComponentMask(None=TextureCoordinate()) {g=true}
  Max(A=Multiply(A=@n51, B=@n51), B=@n27.R)
  Multiply(B=@n40)
  Multiply(A=@n30, B=@n29.PostMult)
  n52 = ComponentMask(None=TextureCoordinate()) {g=true}
  Multiply(A=@n49.Base Color, B=@n52)
  Multiply(A=@n14, B=ComponentMask(None=TextureCoordinate()) {g=true})
  OneMinus(None=@n52)
  OneMinus(None=@n27.R)
  ParticleColor()
  n53 = SetMaterialAttributes!(MaterialAttributes=@n49.MaterialAttributes, Base Color=@n49.MaterialAttributes) {attributeSetTypes [#2]}
  n54 = GetMaterialAttributes(None=MaterialFunctionCall!(
    Override LDR imput with particle color
    In=MaterialFunctionCall!(In=@n53) {materialFunction=/Game/Effects/Niagara/MasterMaterials/Functions/Opacity/FX_DepthFade}."Opacity (MA)"
    Correct Coloring
   ) {materialFunction=/Game/Effects/Niagara/MasterMaterials/Functions/MF_FX_Colorize}."Color (MA)") {attributeGetTypes [#3]}
  n55 = TextureCoordinate()
  n56 = Multiply(A=OneMinus(None=ComponentMask(None=@n55) {g=true})) {constB=0.1}
  n57 = GetMaterialAttributes(None=SetMaterialAttributes!(
    MaterialAttributes=@n54.MaterialAttributes
    Emissive Color=MaterialFunctionCall(Hue Shift Percentage=@n56, Texture=@n54.Emissive Color) {materialFunction=&HueShift}.Result
   ) {attributeSetTypes [#3]}) {attributeGetTypes [#3]}
  n58 = LinearInterpolate(A=EyeAdaptation().EyeAdaptation, Alpha=Saturate(None=ScalarParameter() {parameterName=Eye Adaption Weight}))
  SetMaterialAttributes(
   MaterialAttributes=@n57.MaterialAttributes
   Emissive Color=LinearInterpolate(
    A=@n57.Emissive Color
    B=Divide(A=@n57.Emissive Color, B=@n58)
    Alpha=ScalarParameter() {parameterName=Use Eye Adaption Exposure Control for Emissives}
   )
  ) {attributeSetTypes [#3]}
 ]
}
