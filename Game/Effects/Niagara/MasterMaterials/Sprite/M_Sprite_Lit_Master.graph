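# Master lit, masked material for Niagara sprite particles, with a four-by-one falling-leaf texture as the default BaseTexture parameter.
# The texture color passes through an adjustable Desaturate parameter and is multiplied by particle color, while texture alpha is multiplied by particle alpha.
# Scalar Specular and Roughness parameters expose the Default Lit surface response for tuning.
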Material /Game/Effects/Niagara/MasterMaterials/Sprite/M_Sprite_Lit_Master
graph {
 properties {bUsedWithNiagaraSprites=true, blendMode=BLEND_Masked, pixelDepthOffsetMode=PDOM_Legacy, usedShadingModels=Default Lit}
 nodes [
  n1 = TextureSampleParameter2D() {group=Color, parameterName=BaseTexture, texture=/Game/Effects/Fort_Effects/World/Enviroment/Falling_Leaf/Textures/T_Leaves_Large_4x1}
  n2 = ParticleColor()
  Multiply(A=Desaturation(None=@n1.RGB, Fraction=ScalarParameter() {parameterName=Desaturate, sliderMax=1}), B=@n2.RGB)
  Multiply(A=@n2.A, B=@n1.A)
  ScalarParameter() {group=Specular, parameterName=Specular, sliderMax=1}
  ScalarParameter() {group=Specular, parameterName=Roughness, sortPriority=1}
 ]
}
