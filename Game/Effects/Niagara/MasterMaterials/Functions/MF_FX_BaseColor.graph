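# Material function for Niagara effect master materials that adds a texture sample to the Base Color of incoming material attributes.
# MF_NFX_TextureUVPanner, labeled the Niagara Bridge, builds the UVs from two texture coordinate channels plus parameters for channel choice, frequency and pan speed.
# The sampled RGB is added to the existing Base Color rather than replacing it, then output as material attributes.
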MaterialFunction /Game/Effects/Niagara/MasterMaterials/Functions/MF_FX_BaseColor
graph {
 properties {
  description=This reads in a texture to Base Color. The UVs are controled by the Niagara MaterialUV Module.
  libraryCategoriesText [FX Base Color]
 }
 nodes [
  n1 = GetMaterialAttributes(None=FunctionInput() {inputType=FunctionInput_MaterialAttributes}) {attributeGetTypes [#1]}
  n2 = TextureSampleParameter2D(UVs=MaterialFunctionCall!(
    UV Channel 2=TextureCoordinate() {coordinateIndex=1}
    UV Channel 1=TextureCoordinate()
    In=@n1.MaterialAttributes
    UV Channel=ScalarParameter() {group="2) BaseColor", parameterName="UV Channel (BaseColor)", sliderMax=1, sortPriority=1}
    U Freq Multiplier=ScalarParameter() {defaultValue=1, group="2) BaseColor", parameterName="U Freq (BaseColor)", sortPriority=2}
    V Freq Multiplier=ScalarParameter() {defaultValue=1, group="2) BaseColor", parameterName="V Freq (BaseColor)", sortPriority=3}
    Pan Speed U=ScalarParameter() {group="2) BaseColor", parameterName="Pan U Speed (BaseColor)", sortPriority=4}
    Pan Speed V=ScalarParameter() {group="2) BaseColor", parameterName="Pan V Speed (BaseColor)", sortPriority=5}
   ) {desc=The Niagara Bridge, materialFunction=/Game/Effects/Niagara/MasterMaterials/Functions/UVs/MF_NFX_TextureUVPanner}.Result) {
   desc=This texture provides the foundation for the other channels like opacity and emissive color.
   group="2) BaseColor"
   parameterName=Texture
   texture=/Game/Effects/Fort_Effects/Textures/T_InkySmokeTrail
  }
  FunctionOutput(None=SetMaterialAttributes!(MaterialAttributes=@n1.MaterialAttributes, BaseColor=Add(A=@n1.BaseColor, B=@n2.RGB), EmissiveColor) {attributeSetTypes [#1, #2]}) {outputName="BaseColor (MA)"}
 ]
}
