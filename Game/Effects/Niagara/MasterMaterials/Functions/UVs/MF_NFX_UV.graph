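# Niagara FX master material function that scales UV channels 0 and 1 by frequencies from MF_NFX_Bridge, output as material attributes.
# Both axes of each channel use that channel's U Freq value, leaving no separate V scaling.
# It multiplies CustomizedUV0 and CustomizedUV1 in a chained Get/SetMaterialAttributes pass.
# Scaled texture coordinates, padded with a constant, are also written into BentNormal for UV0 and SubsurfaceColor for UV1.
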MaterialFunction /Game/Effects/Niagara/MasterMaterials/Functions/UVs/MF_NFX_UV
graph {
 properties {bExposeToLibrary=true, libraryCategoriesText [FX Master Function]}
 nodes [
  n1 = GetMaterialAttributes() {attributeGetTypes [#1, #2]}
  n2 = MaterialFunctionCall() {materialFunction=/Game/Effects/Niagara/MasterMaterials/Functions/MF_NFX_Bridge}
  n3 = AppendVector(A=@n2.UV0 U Freq, B=@n2.UV0 U Freq)
  n4 = Multiply(A=TextureCoordinate(), B=@n3)
  n5 = Constant()
  n6 = GetMaterialAttributes(None=SetMaterialAttributes!(MaterialAttributes=@n1.MaterialAttributes, CustomizedUV0=Multiply(A=@n1.CustomizedUV0, B=@n3), BentNormal=AppendVector(A=@n4, B=@n5)) {attributeSetTypes [#1, #3]}) {attributeGetTypes [#2]}
  n7 = AppendVector(A=@n2.UV1 U Freq, B=@n2.UV1 U Freq)
  n8 = Multiply(A=TextureCoordinate() {coordinateIndex=1}, B=@n7)
  n9 = Constant()
  FunctionOutput(None=SetMaterialAttributes!(MaterialAttributes=@n6.MaterialAttributes, CustomizedUV1=Multiply(A=@n6.CustomizedUV1, B=@n7), SubsurfaceColor=AppendVector(A=@n8, B=@n9)) {attributeSetTypes [#2, #4]}) {outputName="CustomizedUVs (MA)"}
 ]
}
