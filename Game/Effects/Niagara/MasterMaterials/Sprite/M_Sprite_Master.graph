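# Master translucent, lit material for Niagara sprite and mesh particles, built on material attributes from shared FX material functions.
# Base color, color correction and opacity functions run in a chain.
# The opacity result branches into a colorize function and a parametric opacity bevel, alongside a separate Niagara UV function.
# Scalar parameters let particle color override the base color texture and control clamped particle tinting ahead of LDR-to-HDR conversion.
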Material /Game/Effects/Niagara/MasterMaterials/Sprite/M_Sprite_Master
graph {
 properties {
  bUseMaterialAttributes=true
  bUsedWithNiagaraMeshParticles=true
  bUsedWithNiagaraSprites=true
  blendMode=BLEND_Translucent
  pixelDepthOffsetMode=PDOM_Legacy
  usedShadingModels=Default Lit
 }
 nodes [
  n1 = GetMaterialAttributes()
  n2 = MaterialFunctionCall!(
   Distortion
   In=MaterialFunctionCall!(In=MaterialFunctionCall!(In=@n1.MaterialAttributes) {materialFunction=/Game/Effects/Niagara/MasterMaterials/Functions/MF_FX_BaseColor}."BaseColor (MA)") {materialFunction=/Game/Effects/Niagara/MasterMaterials/Functions/MF_Niagara_ColorCorrect}."BaseColor (MA)"
  ) {materialFunction=/Game/Effects/Niagara/MasterMaterials/Functions/MF_FX_Opacity}
  n3 = ScalarParameter() {
   desc=This parameter multiplies the texture by particle color and clamps it before going into the ldr range of the ldr to hdr function.
   group="5) Emissive"
   parameterName=Proper Particle HDR Coloring
   sliderMax=1
  }
  MaterialFunctionCall!(
   Override LDR imput with particle color=ScalarParameter() {group="2) BaseColor", parameterName=Override base color texture with particle color}
   In=@n2.Opacity
   Correct Coloring=@n3
  ) {materialFunction=/Game/Effects/Niagara/MasterMaterials/Functions/MF_FX_Colorize}
  MaterialFunctionCall() {materialFunction=/Game/Effects/Niagara/MasterMaterials/Functions/UVs/MF_NFX_UV}
  MaterialFunctionCall!(In=@n2.Opacity) {materialFunction=/Game/Effects/Niagara/MasterMaterials/Functions/Opacity/MF_Parametric_OpacityBevel}
  SetMaterialAttributes!(MaterialAttributes)
 ]
}
