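# Additive unlit master material for Niagara sprites and ribbons, drawing a procedural soft particle with a highlight layer and glow.
# Two sphere masks, driven by two dynamic parameters for hardness and size, combine via Max, with glow scaled by GlowSTR, then multiply by particle alpha.
# Particle color passes through MF_FX_EmissiveAdjustments.
# Each layer uses XSize for both axes, leaving YSize unused; only highlight UVs scale around the center.
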Material /Game/Effects/Niagara/MasterMaterials/Sprite/M_Sprite_ParticleWithSoftGlow_Master
graph {
 properties {
  bEnableResponsiveAA=true
  bOutputTranslucentVelocity=true
  bUsedWithNiagaraRibbons=true
  bUsedWithNiagaraSprites=true
  blendMode=BLEND_Additive
  opacityMaskClipValue=0.03
  pixelDepthOffsetMode=PDOM_Legacy
  shadingModel=MSM_Unlit
  twoSided=true
  usedShadingModels=Unlit
 }
 nodes [
  n1 = ParticleColor()
  MaterialFunctionCall(Emissive=@n1.RGB) {materialFunction=&MF_FX_EmissiveAdjustments}
  n2 = DynamicParameter() {
   parameterIndex=1
   defaultValue=<rgba 0.6 0.1 0.5 1>
   paramNames [HighLightLayer_Hardness, HighLightLayer_XSize, HighLightLayer_YSize, Param4]
  }
  n3 = MaterialFunctionCall(UVs=TextureCoordinate(), Texture Scale=AppendVector(A=@n2.HighLightLayer_XSize, B=@n2.HighLightLayer_XSize)) {materialFunction=&ScaleUVsByCenter}
  n4 = Constant() {r=0.5}
  n5 = DynamicParameter() {defaultValue=<rgba 0 1 1 0.1>, paramNames [GlowLayer_Hardness, GlowLayer_XSize, GlowLayer_YSize, GlowSTR]}
  n6 = SphereMask(
   A=Multiply(A=TextureCoordinate(), B=AppendVector(A=@n5.GlowLayer_XSize, B=@n5.GlowLayer_XSize))
   B=@n4
   Radius=@n4
   Hardness=@n5.GlowLayer_Hardness
  )
  Multiply(A=@n1.A, B=Max(A=SphereMask(A=@n3.UVs, B=@n4, Radius=@n4, Hardness=@n2.HighLightLayer_Hardness), B=Multiply(A=@n6, B=@n5.GlowSTR)))
 ]
}
