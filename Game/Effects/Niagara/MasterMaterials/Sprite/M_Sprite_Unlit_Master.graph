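# Master unlit translucent material for Niagara sprites, built with material attributes.
# Color and alpha textures each get panning normal-map UV distortion scaled by dynamic parameters, then channel masking and particle color tinting.
# Color feeds emissive through optional HDR remapping and adjustment functions; alpha gets a power curve and optional depth fade.
# A skydive camera-fade function follows, and world position offset collapses sprites toward their particle centers near the camera.
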Material /Game/Effects/Niagara/MasterMaterials/Sprite/M_Sprite_Unlit_Master
graph {
 properties {
  bUseMaterialAttributes=true
  bUsedWithNiagaraSprites=true
  blendMode=BLEND_Translucent
  pixelDepthOffsetMode=PDOM_Legacy
  shadingModel=MSM_Unlit
  usedShadingModels=Unlit
 }
 nodes [
  MakeMaterialAttributes!(
   BaseColor
   Metallic
   Specular
   Roughness
   Anisotropy
   EmissiveColor
   Opacity
   OpacityMask
   Normal
   Tangent
   WorldPositionOffset
   SubsurfaceColor
   ClearCoat
   ClearCoatRoughness
   AmbientOcclusion
   Refraction
   CustomizedUVs_0
   CustomizedUVs_1
   CustomizedUVs_2
   CustomizedUVs_3
   CustomizedUVs_4
   CustomizedUVs_5
   CustomizedUVs_6
   CustomizedUVs_7
   PixelDepthOffset
   ShadingModel
   Displacement
  )
  n1 = TextureCoordinate()
  n2 = VectorParameter() {group="2) Color", parameterName=Color Distortion UV Tile And Speed, defaultValue=<rgba 1 1 0 0>}
  n3 = ComponentMask(None=@n2.RGB) {g=true, r=true}
  n4 = AppendVector(A=@n2.B, B=@n2.B)
  n5 = Time()
  n6 = TextureSampleParameter2D(UVs=Add(A=Multiply(A=@n1, B=@n3), B=Multiply(A=@n4, B=@n5))) {parameterName=Distortion Texture, samplerType=SAMPLERTYPE_Normal, texture=/Game/Global/Textures/Default/Tileable_Patterns/1Normal}
  n7 = ComponentMask(None=@n6.RGB) {g=true, r=true}
  n8 = DynamicParameter() {paramNames [ColorDistortionMultiplier, OpacityDistortionMultiplier, Param3, Param4]}
  n9 = Multiply(A=ScalarParameter() {group="2) Color", parameterName=Color Distortion Intensity Multiplier}, B=@n8.ColorDistortionMultiplier)
  n10 = TextureSampleParameter2D(UVs=Add(A=TextureCoordinate(), B=Multiply(A=@n7, B=@n9))) {parameterName=Color Texture, samplerSource=SSM_Wrap_WorldGroupSettings, texture=/Game/Effects/Fort_Effects/Textures/Explo/T_Fire_Ring}
  n11 = Multiply(
   A=LinearInterpolate(
    A=@n10.RGB
    B=ChannelMaskParameter(None=@n10.RGB) {group="2) Color", parameterName=Color Component Mask, sortPriority=1}.R
    Alpha=ScalarParameter() {defaultValue=1, group="2) Color", parameterName=UseColorComponentMasking}
   )
   B=ParticleColor().RGB
  )
  n12 = Saturate(None=@n11)
  n13 = MaterialFunctionCall(LDR Input=@n12, HDR Tint=@n12, HDR Intensity Multiplier=ScalarParameter() {defaultValue=1, parameterName=HDR Multiplier}) {materialFunction=&DeriveHDRfromLDR}
  n14 = SetMaterialAttributes!(
   MaterialAttributes
   Base Color=@n11
   Emissive Color=MaterialFunctionCall(Emissive=StaticSwitchParameter(True=@n13.RemappedRange, False=@n11) {parameterName=UseHDR}) {materialFunction=&MF_FX_EmissiveAdjustments}.Result
  ) {attributeSetTypes [#1, #2]}
  n15 = TextureCoordinate()
  n16 = VectorParameter() {group="3) Alpha", parameterName=Opacity Distortion UV Tile And Speed, defaultValue=<rgba 1 1 0 0>}
  n17 = ComponentMask(None=@n16.RGB) {g=true, r=true}
  n18 = AppendVector(A=@n16.B, B=@n16.B)
  n19 = Time()
  n20 = TextureSampleParameter2D(UVs=Add(A=Multiply(A=@n15, B=@n17), B=Multiply(A=@n18, B=@n19))) {parameterName=Distortion Texture, samplerType=SAMPLERTYPE_Normal, texture=&1Normal}
  n21 = ComponentMask(None=@n20.RGB) {g=true, r=true}
  n22 = Multiply(A=ScalarParameter() {group="3) Alpha", parameterName=Opacity Distortion Intensity Multiplier}, B=@n8.OpacityDistortionMultiplier)
  n23 = TextureSampleParameter2D(UVs=Add(A=TextureCoordinate(), B=Multiply(A=@n21, B=@n22))) {parameterName=Alpha Texture, samplerSource=SSM_Wrap_WorldGroupSettings, texture=&T_Fire_Ring}
  n24 = ChannelMaskParameter(None=@n23.RGBA) {group="3) Alpha", parameterName=Alpha Texture Component Mask, sortPriority=1}
  n25 = ScalarParameter() {defaultValue=1, group="3) Alpha", parameterName=Alpha Texture Power}
  n26 = Saturate(None=Multiply(A=Power(Base=@n24.R, Exp=@n25), B=ParticleColor().A))
  n27 = StaticSwitchParameter(True=DepthFade(Opacity=@n26, FadeDistance=ScalarParameter() {defaultValue=150, parameterName=Depth Fade Distance}), False=@n26) {parameterName=UseDepthFade}
  n28 = ParticlePositionWS()
  n29 = WorldPosition()
  n30 = Distance(A=@n28, B=CameraPositionWS())
  n31 = ScalarParameter() {defaultValue=150, group="8) WPO", parameterName=Sprite Shrink Camera Distance, sortPriority=80}
  n32 = Saturate(None=Divide(
    A=Subtract(A=@n30, B=@n31)
    B=ScalarParameter() {defaultValue=350, group="8) WPO", parameterName=Sprite Shrink Distance Falloff Hardness, sortPriority=81}
   ))
  SetMaterialAttributes!(
   MaterialAttributes=MaterialFunctionCall!(Material Attributes=SetMaterialAttributes!(MaterialAttributes=@n14, Opacity=@n27) {attributeSetTypes [#3]}) {materialFunction=/Game/Effects/Niagara/MasterMaterials/Functions/V2/MF_Skydive_CameraFade_V2}."Opacitys and BaseColor (MA)"
   World Position Offset=Multiply(A=Subtract(A=@n28, B=@n29.XYZ), B=OneMinus(None=@n32))
  ) {attributeSetTypes [#4]}
 ]
}
