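# Unlit translucent master material for Niagara sprite and mesh particles that draws a soft square shape procedurally, without textures.
# Particle color RGB feeds the MF_FX_EmissiveAdjustments function as emissive.
# Sines of half-scaled U and V coordinates are multiplied and squared, forming a mask that peaks at the center and fades toward the edges.
# That mask is multiplied by particle alpha.
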Material /Game/Effects/Niagara/MasterMaterials/Sprite/M_Square_Master
graph {
 properties {
  bUsedWithNiagaraMeshParticles=true
  bUsedWithNiagaraSprites=true
  blendMode=BLEND_Translucent
  pixelDepthOffsetMode=PDOM_Legacy
  shadingModel=MSM_Unlit
  usedShadingModels=Unlit
 }
 nodes [
  n1 = ParticleColor()
  MaterialFunctionCall(Emissive=@n1.RGB) {materialFunction=&MF_FX_EmissiveAdjustments}
  n2 = TextureCoordinate() {uTiling=0.5, vTiling=0.5}
  n3 = Multiply(A=Sine(None=ComponentMask(None=@n2) {r=true}), B=Sine(None=ComponentMask(None=@n2) {g=true}))
  Multiply(A=@n1.A, B=Multiply(A=@n3, B=@n3))
  Multiply()
 ]
}
