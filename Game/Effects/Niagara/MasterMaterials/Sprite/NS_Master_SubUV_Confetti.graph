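# Niagara system, named as a master SubUV confetti sprite effect, that spawns small falling sprites from a single emitter named "Empty".
# Spawn Rate emits particles continuously, and Initialize Particle sets size and rotation and writes a material random value before Shape Location places them.
# Gravity Force and Solve Forces and Velocity pull them downward, with no initial velocity module.
# A sprite renderer draws them; no SubUV animation module or user parameters appear.
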NiagaraSystem /Game/Effects/Niagara/MasterMaterials/Sprite/NS_Master_SubUV_Confetti
dependencies {usedDataInterfaces [], usedDynamicInputs [], usedModules [&SystemState]}
emitters {Empty {
  emitterUpdateScript [
   &EmitterState(Life Cycle Mode=::NewEnumerator1, Loop Duration=1, UseLoopDelay=false)
   &SpawnRate(SpawnRate=10, Use Spawn Probability=false)
  ]
  particleSpawnScript [
   &InitializeParticle(
    Color Mode=::NewEnumerator0
    Sprite Rotation Mode=::NewEnumerator2
    Sprite Size Mode=::NewEnumerator1
    Uniform Sprite Size Max=30
    Uniform Sprite Size Min=10
    UsePositionOffset=false
    Write Material Random=true
   )
   &ShapeLocation(MaskSpawnGroup=false, Override Seed=false, UseOwnerScale=true)
  ]
  particleUpdateScript [
   &ParticleState()
   &GravityForce(Gravity=<0 0 -200>)
   &SolveForcesAndVelocity(Clamp Velocity=false, Limit Acceleration=false)
  ]
  rendererClasses [&NiagaraSpriteRendererProperties]
  renderers [{rendererClass=&NiagaraSpriteRendererProperties}]
 }}
summary {
 systemName=NS_Master_SubUV_Confetti
 emitters [{emitterName=Empty, rendererClasses [&NiagaraSpriteRendererProperties]}]
 userVariables []
}
