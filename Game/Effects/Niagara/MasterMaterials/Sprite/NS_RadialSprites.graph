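# Niagara system with a single "Rays" sprite emitter that produces light-blue radial sprites, driven through radial material controls.
# A steady low spawn rate creates short-lived, uniformly sized particles with material randomness enabled.
# In update, FN_Radial_Material sets tiling, mask, extrusion and spin, using random ranges for tiling and spin speed.
# Curves grow each sprite and fade its alpha to zero over its life.
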NiagaraSystem /Game/Effects/Niagara/MasterMaterials/Sprite/NS_RadialSprites
dependencies {usedDataInterfaces [&NiagaraDataInterfaceCurve], usedDynamicInputs [], usedModules [&SystemState]}
emitters {Rays {
  emitterUpdateScript [
   &EmitterState(Life Cycle Mode=::NewEnumerator1, Loop Duration=1, UseLoopDelay=false)
   &SpawnRate(SpawnRate=2, Use Spawn Probability=false)
  ]
  particleSpawnScript [&InitializeParticle(
    Color=<rgba 0.1719 0.6146 1 1>
    Lifetime=1.25
    Sprite Rotation Mode=::NewEnumerator2
    Sprite Size Mode=::NewEnumerator3
    Uniform Sprite Size=50
    UsePositionOffset=false
    Write Material Random=true
   )]
  particleUpdateScript [
   &ParticleState()
   &ScaleSpriteSize(Uniform Curve Sprite Scale=DataInterface() {propertyValues {Curve {keys [{interpMode=RCIM_Cubic, tangentMode=RCTM_User, time=0, value=0.7}, {interpMode=RCIM_Cubic, tangentMode=RCTM_User, time=1, value=1.7}]}}})
   &FN_Radial_Material(
    Base Sprite Tile U:NiagaraFloat=&RandomRangeFloat(Minimum=1.2, Maximum=2)
    Radial Mask Power=3
    Radial Mask Strength=0.8
    Sprite Extrusion Speed=0.5
    Sprite Spin Speed:NiagaraFloat=&RandomRangeFloat(Minimum=0.2, Maximum=0.5)
   )
   &ScaleColor(
    Scale Alpha=&FloatFromCurve(FloatCurve=DataInterface() {propertyValues {Curve {keys [{interpMode=RCIM_Cubic, tangentMode=RCTM_User, time=0, value=1}, {interpMode=RCIM_Cubic, tangentMode=RCTM_User, time=1, value=0}]}}})
    ScaleA=true
    ScaleRGB=false
   )
  ]
  rendererClasses [&NiagaraSpriteRendererProperties]
  renderers [{rendererClass=&NiagaraSpriteRendererProperties}]
 }}
summary {
 systemName=NS_RadialSprites
 emitters [{emitterName=Rays, rendererClasses [&NiagaraSpriteRendererProperties]}]
 userVariables []
}
